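# Unlit debug material for Niagara sprites, in the Flower Skeleton backpack VFX folder, that displays a dynamic parameter value as digits.
# Param1 feeds both a DebugScalarValues readout and, copied to all four channels, a color-coded DebugFloat4Values readout.
# "Track one Value" blends from four-channel to scalar, defaulting to scalar, before MF_FX_EmissiveAdjustments.
# The inner lerp on the four-channel readout has identical inputs, so it changes nothing.
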Material /Game/Accessories/FORT_Backpacks/Backpack_M_MED_Flower_Skeleton/VFX/M_DEBUG_DynamicPassthrough
graph {
 properties {bUsedWithNiagaraSprites=true, pixelDepthOffsetMode=PDOM_Legacy, shadingModel=MSM_Unlit, usedShadingModels=Unlit}
 nodes [
  n1 = DynamicParameter()
  n2 = ScalarParameter() {defaultValue=3, parameterName="MaximumNumberOfDigits (S)"}
  n3 = MaterialFunctionCall!(
   Vector4=MaterialFunctionCall(X=@n1.Param1, Y=@n1.Param1, Z=@n1.Param1, A=@n1.Param1) {materialFunction=&MakeFloat4}.Result
   MaximumNumberOfDigits=@n2
   UVs
   DebugTextLocation RG_UpperRight BA_LowerLeft=Constant4Vector() {constant=<rgba 0 0 1 1>}.RGBA
   Component Spacing
  ) {materialFunction=/Engine/Functions/Engine_MaterialFunctions02/Utility/DebugFloat4Values}
  MaterialFunctionCall(Emissive=LinearInterpolate(
    A=LinearInterpolate(A=@n3.ColorCodedOutput, B=@n3.ColorCodedOutput) {constAlpha=0.25}
    B=MaterialFunctionCall(Number=@n1.Param1, MaximumNumberOfDigits=@n2) {materialFunction=&DebugScalarValues}.Result
    Alpha=ScalarParameter() {defaultValue=1, parameterName=Track one Value}
   )) {materialFunction=&MF_FX_EmissiveAdjustments}
 ]
}
